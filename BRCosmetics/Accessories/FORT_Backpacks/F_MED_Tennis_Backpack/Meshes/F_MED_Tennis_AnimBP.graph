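# Animation Blueprint for the F_MED_Tennis backpack that nudges its root bone to follow the wearer's spine bending and twisting.
# Each update, it measures the spine_05 rotation change from reference pose relative to spine_01 on the owner's skeletal mesh component.
# Clamped range mapping turns this into three weights: rotate from yaw in either direction, plus separate right and left twists.
# These weights drive three chained additive bone-space Transform Bone nodes on backpack_root.
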AnimBlueprint /BRCosmetics/Accessories/FORT_Backpacks/F_MED_Tennis_Backpack/Meshes/F_MED_Tennis_AnimBP
target_skeleton=&/F_MED_Tennis_Skeleton
graphs [
 AnimGraph {nodes [
   n1 = "Transform (Modify) Bone - Bone: backpack_root"(
    ComponentPose=Local To Component(LocalPose=Input Pose())
    Alpha=Get backpack_rotate() -> (backpack_rotate:double=0)
    Translation="0,-3,0"
    Rotation="0,10,0"
    Scale="1,1,1"
   ) {
    rotationMode=BMM_Additive
    rotationSpace=BCS_BoneSpace
    translationMode=BMM_Additive
    translationSpace=BCS_BoneSpace
    alphaBoolBlend {blendOption=Linear}
    boneToModify {boneName=backpack_root}
    rotation {yaw=10}
    translation {y=-3}
   }
   n2 = "Transform (Modify) Bone - Bone: backpack_root"(
    ComponentPose=@n1
    Alpha=Get backpack_twist_right() -> (backpack_twist_right:double)
    Translation="0,-3,5"
    Rotation="10,0,15"
    Scale="1,1,1"
   ) {
    rotationMode=BMM_Additive
    rotationSpace=BCS_BoneSpace
    translationMode=BMM_Additive
    translationSpace=BCS_BoneSpace
    alphaBoolBlend {blendOption=Linear}
    boneToModify {boneName=backpack_root}
    rotation {pitch=10, roll=15}
    translation {y=-3, z=5}
   }
   n3 = "Transform (Modify) Bone - Bone: backpack_root"(ComponentPose=@n2, Alpha=Get backpack_twist_left() -> (backpack_twist_left:double), Translation="0,-3,0", Rotation="15,0,0", Scale="1,1,1") {
    rotationMode=BMM_Additive
    rotationSpace=BCS_BoneSpace
    translationMode=BMM_Additive
    translationSpace=BCS_BoneSpace
    alphaBoolBlend {blendOption=Linear}
    boneToModify {boneName=backpack_root}
    rotation {pitch=15}
    translation {y=-3}
   }
   Output Pose(Result=Component To Local(ComponentPose=@n3)) {layerGroup=None}
  ]}
 EventGraph {nodes [
   Event BlueprintUpdateAnimation()
   n1 = GetOwnerSkeletalMeshComponent()
   n2 = BreakRotator(InRot=BreakTransform(InTransform=GetDeltaTransformFromRefPose(self=@n1, BoneName=spine_05, BaseName=spine_01)).Rotation)
   n3 = MapRangeClamped(Value=@n2.Yaw, InRangeA=0, InRangeB=50)
   n4 = MapRangeClamped(Value=@n2.Yaw, InRangeA=0, InRangeB=-30)
   > Set backpack_rotate(backpack_rotate:double=SelectFloat(A=@n4, B=@n3, bPickA="float > float"(A=@n4, B=@n3) {is=title K2Node_CallFunction})) -> (Output_Get:double=0)
   > Set backpack_twist_right(backpack_twist_right:double=MapRangeClamped(Value=@n2, InRangeA=0, InRangeB=55)) -> (Output_Get:double)
   > Set backpack_twist_left(backpack_twist_left:double=MapRangeClamped(Value=@n2, InRangeA=0, InRangeB=-60)) -> (Output_Get:double)
  ]}
]
variables [
 backpack_rotate {type=number}
 backpack_twist_left {type=number}
 backpack_twist_right {type=number}
]
